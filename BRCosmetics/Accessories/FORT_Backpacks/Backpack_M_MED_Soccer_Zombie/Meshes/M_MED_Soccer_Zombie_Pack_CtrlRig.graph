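# Control Rig for the Soccer Zombie backpack cosmetic that procedurally animates its upper and lower jaws.
# On execution, two Chain Harmonics units run in sequence on the three-bone lower and upper jaw chains.
# Each uses only noisy wave motion with Z-axis amplitude, the lower jaw at a faster speed, with pendulum and reach disabled.
# Slime and drip bones exist in the hierarchy but are not driven here, and debug drawing is left on.
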ControlRigBlueprint /BRCosmetics/Accessories/FORT_Backpacks/Backpack_M_MED_Soccer_Zombie/Meshes/M_MED_Soccer_Zombie_Pack_CtrlRig
graphs [
 RigVMFunctionLibrary {links [], nodes []}
 RigVMModel {nodes [
   n1 = &RigUnit_BeginExecution()
   /Script/ControlRig.RigUnit_ChainHarmonics(
    ChainRoot=dyn_jaw_lower_02
    DrawWorldOffset=(Rotation=(X=0,Y=0,Z=0,W=1),Translation=(X=0,Y=0,Z=0),Scale3D=(X=1,Y=1,Z=1))
    Pendulum=(bEnabled=False,PendulumStiffness=2,PendulumGravity=(X=0,Y=0,Z=0),PendulumBlend=0.75,PendulumDrag=0.98,PendulumMinimum=0,PendulumMaximum=0.1,PendulumEase=Linear,UnwindAxis=(X=0,Y=1,Z=0),UnwindMinimum=0.2,UnwindMaximum=0.05)
    Reach=(bEnabled=False,ReachTarget=(X=0,Y=0,Z=0),ReachAxis=(X=1,Y=0,Z=0),ReachMinimum=0,ReachMaximum=0,ReachEase=SineEaseInOut)
    Speed=(X=0,Y=0,Z=1.55)
    Wave=(bEnabled=True,WaveFrequency=(X=1,Y=1,Z=1),WaveAmplitude=(X=0,Y=0,Z=0.45),WaveOffset=(X=1,Y=1,Z=5),WaveNoise=(X=1,Y=1,Z=3),WaveMinimum=1,WaveMaximum=2,WaveEase=SineEaseInOut)
    WaveCurve=(EditorCurveData=(Keys=((),(Time=1,Value=1)),PreInfinityExtrap=RCCE_Constant,PostInfinityExtrap=RCCE_Constant,DefaultValue=340282346638528859811704183484516925440))
    bDrawDebug=True
    ExecutePin=@n1.ExecutePin
   ) {is=struct RigVMUnitNode}
   > &RigUnit_ChainHarmonics(
    ChainRoot=dyn_jaw_upper_02
    DrawWorldOffset=(Rotation=(X=0,Y=0,Z=0,W=1),Translation=(X=0,Y=0,Z=0),Scale3D=(X=1,Y=1,Z=1))
    Pendulum=(bEnabled=False,PendulumStiffness=2,PendulumGravity=(X=0,Y=0,Z=0),PendulumBlend=0.75,PendulumDrag=0.98,PendulumMinimum=0,PendulumMaximum=0.1,PendulumEase=Linear,UnwindAxis=(X=0,Y=1,Z=0),UnwindMinimum=0.2,UnwindMaximum=0.05)
    Reach=(bEnabled=False,ReachTarget=(X=0,Y=0,Z=0),ReachAxis=(X=1,Y=0,Z=0),ReachMinimum=0,ReachMaximum=0,ReachEase=SineEaseInOut)
    Speed=(X=0,Y=0,Z=0.6)
    Wave=(bEnabled=True,WaveFrequency=(X=1,Y=1,Z=1),WaveAmplitude=(X=0,Y=0,Z=0.5),WaveOffset=(X=1,Y=1,Z=5),WaveNoise=(X=1,Y=1,Z=2),WaveMinimum=1,WaveMaximum=2,WaveEase=SineEaseInOut)
    WaveCurve=(EditorCurveData=(Keys=((),(Time=1,Value=1)),PreInfinityExtrap=RCCE_Constant,PostInfinityExtrap=RCCE_Constant,DefaultValue=340282346638528859811704183484516925440))
    bDrawDebug=True
   )
  ]}
]
hierarchy [
 root {loc=<0 0 0>, rot=<0 0 0 1>}
 pack {parent=root, loc=<0 -0.7316 0>, rot=<0 0 0 1>}
 dyn_slime_parent {parent=pack, loc=<0.5612 -17.5384 -3.3504>, rot=<-0.5 -0.5 0.5 -0.5>}
 dyn_slime_04 {parent=dyn_slime_parent, loc=<2.8777 -4.123 -0.3876>, rot=<0 0 0 1>}
 dyn_slime_02 {parent=dyn_slime_parent, loc=<0.4749 4.6559 -0.6932>, rot=<0 0 0 1>}
 dyn_slime_03 {parent=dyn_slime_02, loc=<2.3752 -0.4166 -0.0869>, rot=<0 0 0 1>}
 dyn_slime_01 {parent=dyn_slime_parent, loc=<-0.4171 6.6665 -1.2269>, rot=<0 0 0 1>}
 dyn_drip_parent {parent=pack, loc=<-5.5276 -11.858 -8.8883>, rot=<-0.5 -0.5 0.5 -0.5>}
 dyn_drip_01 {parent=dyn_drip_parent, loc=<2.3676 0.4987 0.5006>, rot=<0 0 0 1>}
 dyn_drip_02 {parent=dyn_drip_parent, loc=<-0.0022 -2.0805 0.8738>, rot=<0 0 0.2924 -0.9563>}
 dyn_jaw_upper_02 {parent=pack, loc=<0.4602 -8.4215 0.5008>, rot=<-0.5 0.5 0.5 -0.5>}
 dyn_jaw_upper_01 {parent=dyn_jaw_upper_02, loc=<1.2228 0 0>, rot=<0 0 0 1>}
 dyn_jaw_upper {parent=dyn_jaw_upper_01, loc=<1.2033 0 0>, rot=<0 0 0.2588 -0.9659>}
 dyn_jaw_lower_02 {parent=pack, loc=<0.4602 -8.3071 -0.7369>, rot=<-0.5 0.5 0.5 -0.5>}
 dyn_jaw_lower_01 {parent=dyn_jaw_lower_02, loc=<1.2964 0 0>, rot=<0 0 0 1>}
 dyn_jaw_lower {parent=dyn_jaw_lower_01, loc=<1.1273 0 0>, rot=<0 0 0.4226 -0.9063>}
]
