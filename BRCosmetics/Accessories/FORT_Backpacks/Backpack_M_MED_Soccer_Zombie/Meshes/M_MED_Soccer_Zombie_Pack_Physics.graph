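# Physics asset for the zombie soccer backpack cosmetic, letting its lower jaw move on the pack.
# Two box bodies cover the dyn_jaw_lower bone and the pack bone.
# A single constraint joins jaw to pack with twist and second swing locked, restricting the jaw to a hinge-like swing on one axis.
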PhysicsAsset /BRCosmetics/Accessories/FORT_Backpacks/Backpack_M_MED_Soccer_Zombie/Meshes/M_MED_Soccer_Zombie_Pack_Physics
bodies {
 dyn_jaw_lower [{shapeType=Box, extentX=5.9637, extentY=1.5835, extentZ=12.7753, center=<2.4543 0 0>}]
 pack [{shapeType=Box, extentX=1.25, extentY=1.25, extentZ=1.25}]
}
constraints [{
  bone1Name=dyn_jaw_lower
  bone2Name=pack
  swing1LimitDegrees=18
  swing2LimitDegrees=0
  swing2Motion=Locked
  twistLimitDegrees=5
  twistMotion=Locked
 }]
