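# Physics asset for the slime version of the zombie soccer backpack cosmetic, giving its dangling slime and drip bones physical bodies.
# Small capsules cover the slime, drip and lower jaw bones, with a box on the pack root.
# Slime and drips are constrained to the pack, except slime_03, which chains off slime_02.
# Drips have looser swing and twist limits than slime, and the jaw body has no constraint.
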PhysicsAsset /BRCosmetics/Accessories/FORT_Backpacks/Backpack_M_MED_Soccer_Zombie/Meshes/M_MED_Soccer_Zombie_Pack_Slime_Physics
bodies {
 dyn_drip_01 [{shapeType=Capsule, length=0.7644, radius=0.6285, center=<0.9405 0 0>, rotation=<rot 90 0 0>}]
 dyn_drip_02 [{shapeType=Capsule, length=0.8656, radius=0.8494, center=<1.2874 0 0>, rotation=<rot 90 0 0>}]
 dyn_jaw_lower [{shapeType=Capsule, length=14.125, radius=2.125, center=<6.3089 -0.4965 0.05>, rotation=<rot 0 0 5.5555>}]
 dyn_slime_01 [{shapeType=Capsule, length=1.505, radius=0.505, center=<1.0759 0 0>, rotation=<rot 90 0 0>}]
 dyn_slime_02 [{shapeType=Capsule, length=1.2517, radius=0.6971, center=<1.2207 0 0>, rotation=<rot 90 0 0>}]
 dyn_slime_03 [{shapeType=Capsule, length=1.1255, radius=0.625, center=<1.1338 0 0>, rotation=<rot 90 0 0>}]
 dyn_slime_04 [{shapeType=Capsule, length=0.8581, radius=0.668, center=<0.9431 0 0>, rotation=<rot 90 0 0>}]
 pack [{shapeType=Box, extentX=1.25, extentY=1.25, extentZ=1.25}]
}
constraints [
 {bone1Name=dyn_slime_02, bone2Name=pack, swing1LimitDegrees=30, swing2LimitDegrees=30, twistLimitDegrees=6}
 {bone1Name=dyn_slime_03, bone2Name=dyn_slime_02, swing1LimitDegrees=30, swing2LimitDegrees=30, twistLimitDegrees=6}
 {bone1Name=dyn_slime_01, bone2Name=pack, swing1LimitDegrees=30, swing2LimitDegrees=30, twistLimitDegrees=6}
 {bone1Name=dyn_slime_04, bone2Name=pack, swing1LimitDegrees=30, swing2LimitDegrees=30, twistLimitDegrees=6}
 {bone1Name=dyn_drip_01, bone2Name=pack, swing1LimitDegrees=50, swing2LimitDegrees=50, twistLimitDegrees=20}
 {bone1Name=dyn_drip_02, bone2Name=pack, swing1LimitDegrees=50, swing2LimitDegrees=50, twistLimitDegrees=20}
]
