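# Physics asset for the upper jaw of the Soccer Zombie backpack cosmetic, giving its dyn_jaw_upper bones bodies for physics-driven motion.
# A larger offset box covers dyn_jaw_upper and a small box covers dyn_jaw_upper_01, joined by a single constraint.
# Twist and second swing are locked, leaving a narrow swing on one axis so the joint acts as a limited hinge.
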PhysicsAsset /BRCosmetics/Accessories/FORT_Backpacks/Backpack_M_MED_Soccer_Zombie/Meshes/M_MED_Soccer_Zombie_Pack_UpperJaw_Physics
bodies {
 dyn_jaw_upper [{shapeType=Box, extentX=4.0976, extentY=1.0803, extentZ=4.6303, center=<1.8313 0.4435 -0.0001>}]
 dyn_jaw_upper_01 [{shapeType=Box, extentX=1.01, extentY=1.01, extentZ=1.01, rotation=<rot 0 -90 0>}]
}
constraints [{
  bone1Name=dyn_jaw_upper
  bone2Name=dyn_jaw_upper_01
  swing1LimitDegrees=15
  swing2LimitDegrees=3
  swing2Motion=Locked
  twistLimitDegrees=3
  twistMotion=Locked
 }]
