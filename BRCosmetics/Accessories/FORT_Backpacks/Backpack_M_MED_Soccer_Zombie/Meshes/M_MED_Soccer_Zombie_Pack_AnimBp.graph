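# Animation Blueprint for the zombie soccer backpack cosmetic, animating its jaws and slime.
# When jumping, falling or moving uncrouched, the cached input pose runs through chained slime, upper-jaw and pack RigidBody simulations, the last with an upward external force.
# Otherwise it uses a Control Rig plus slime physics.
# Constraints then attach drip and slime parent bones to the jaws; the computed Jump_Alpha and Jaw_Jump_Alpha are unused in the AnimGraph.
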AnimBlueprint /BRCosmetics/Accessories/FORT_Backpacks/Backpack_M_MED_Soccer_Zombie/Meshes/M_MED_Soccer_Zombie_Pack_AnimBp
target_skeleton=&/M_MED_Soccer_Zombie_Pack_Skeleton
graphs [
 AnimGraph {nodes [
   n1 = RigidBody(ComponentPose=Local To Component(LocalPose=Use cached pose 'ZomSoccerPackCache'())) {
    bClampLinearTranslationLimitToRefPose=true
    bOverrideWorldGravity=true
    overridePhysicsAsset=&/M_MED_Soccer_Zombie_Pack_Slime_Physics
    simulationSpace=BaseBoneSpace
    alphaBoolBlend {blendOption=Linear}
    baseBoneRef {boneName=root}
    componentAppliedLinearAccClamp=<6000 6000 6000>
    componentLinearAccScale=<0.25 0.25 0.25>
    componentLinearVelScale=<0.25 0.25 0.25>
    overrideWorldGravity {z=-980}
   }
   n2 = RigidBody(ComponentPose=@n1, Alpha=0.75) {
    bClampLinearTranslationLimitToRefPose=true
    bOverrideWorldGravity=true
    overridePhysicsAsset=&/M_MED_Soccer_Zombie_Pack_UpperJaw_Physics
    simulationSpace=BaseBoneSpace
    alphaBoolBlend {blendOption=Linear}
    baseBoneRef {boneName=root}
    componentAppliedLinearAccClamp=<500 500 500>
    componentLinearAccScale=<0.25 0.25 0.25>
    componentLinearVelScale=<0.25 0.25 0.25>
    overrideWorldGravity {z=-600}
   }
   n3 = Get bIsJumping()
   n4 = Get bIsFalling()
   n5 = RigidBody(
    ComponentPose=@n2
    Alpha=SelectFloat(A=0.75, B=Get PawnSpeed2D(), bPickA=OR Boolean(A=@n3, B=@n4))
    ExternalForce=Get Ext_Force() -> (Ext_Force:Vector)
   ) {
    bClampLinearTranslationLimitToRefPose=true
    bOverrideWorldGravity=true
    overridePhysicsAsset=&/M_MED_Soccer_Zombie_Pack_Physics
    simulationSpace=BaseBoneSpace
    alphaBoolBlend {blendOption=Linear}
    baseBoneRef {boneName=root}
    componentAppliedLinearAccClamp=<4000 4000 4000>
    componentLinearAccScale=<0.25 0.25 0.5>
    componentLinearVelScale=<0.25 0.25 0.5>
    overrideWorldGravity {z=-850}
   }
   n6 = Component To Local(ComponentPose=@n5)
   n7 = Control Rig(Source=Use cached pose 'ZomSoccerPackCache'()) {
    alphaBoolBlend {blendOption=Linear}
    controlRigAssetReference {blueprintRigClass=&/M_MED_Soccer_Zombie_Pack_CtrlRig.M_MED_Soccer_Zombie_Pack_CtrlRig_C}
   }
   n8 = RigidBody(ComponentPose=Local To Component(LocalPose=@n7)) {
    bClampLinearTranslationLimitToRefPose=true
    bOverrideWorldGravity=true
    overridePhysicsAsset=&M_MED_Soccer_Zombie_Pack_Slime_Physics
    simulationSpace=BaseBoneSpace
    alphaBoolBlend {blendOption=Linear}
    baseBoneRef {boneName=root}
    componentAppliedLinearAccClamp=<6000 6000 6000>
    componentLinearAccScale=<0.25 0.25 0.25>
    componentLinearVelScale=<0.25 0.25 0.25>
    overrideWorldGravity {z=-980}
   }
   n9 = Component To Local(ComponentPose=@n8)
   n10 = OR Boolean!(A=Get bIsJumping(), B=Get bIsFalling(), C:bool=Get PawnSpeed() -> (PawnSpeed:bool)) -> (ReturnValue)
   n11 = "Constraint - dyn_drip_parent (P)"(
    ComponentPose=Local To Component(LocalPose=Blend Poses by bool(BlendPose_0=@n6, BlendPose_1=@n9, bActiveValue=@n10) {blendType=Linear})
    Alpha=1
    ConstraintWeights_0=1
   ) {
    alphaBoolBlend {blendOption=Linear}
    boneToModify {boneName=dyn_drip_parent}
    constraintSetup [{targetBone {boneName=dyn_jaw_lower}}]
    constraintWeights [1]
   }
   n12 = "Constraint - dyn_slime_parent (P)"(ComponentPose=@n11, Alpha=1, ConstraintWeights_0=1) {
    alphaBoolBlend {blendOption=Linear}
    boneToModify {boneName=dyn_slime_parent}
    constraintSetup [{targetBone {boneName=dyn_jaw_upper}}]
    constraintWeights [1]
   }
   Output Pose(Result=Component To Local(ComponentPose=@n12)) {layerGroup=None}
   AnimGraphNode_SubInput_0 = Input Pose()
   Save cached pose 'ZomSoccerPackCache'(Pose=@AnimGraphNode_SubInput_0) {cacheName=ZomSoccerPackCache}
  ]}
 EventGraph {nodes [
   Event BlueprintUpdateAnimation()
   > n1 = Sequence!(execute) -> (then_0, then_1, then_2:Exec, then_3:Exec)
   > Set Ext_Force(Ext_Force:Vector="0, 0,1250") -> (Output_Get:Vector)
   n2 = Get bIsFalling()
   n3 = Get bIsJumping()
   Set Jaw_Jump_Alpha(execute=@n1.then_2, Jaw_Jump_Alpha:double=SelectFloat(A=0, B=1, bPickA=OR Boolean!(A=@n3, B=@n2) -> (ReturnValue))) -> (Output_Get:double)
   n4 = Get bIsJumping()
   n5 = Get bIsFalling()
   Set Jump_Alpha(execute=@n1.then_1, Jump_Alpha:double=SelectFloat(A=0.65, B=1, bPickA=OR Boolean(A=@n4, B=@n5))) -> (Output_Get:double)
   n6 = Get PawnSpeed2D()
   n7 = Get bIsCrouching()
   Set PawnSpeed(execute=@n1.then_3, PawnSpeed:bool=AND Boolean(A="float > float"(A=@n6, B=1) {is=title K2Node_CallFunction}, B=NOT Boolean(A=@n7))) -> (Output_Get:bool)
   TryGetPawnOwner()
  ]}
]
variables [
 Ext_Force {type=Vector}
 Jaw_Jump_Alpha {type=number}
 Jump_Alpha {type=number}
 PawnSpeed {type=boolean}
]
